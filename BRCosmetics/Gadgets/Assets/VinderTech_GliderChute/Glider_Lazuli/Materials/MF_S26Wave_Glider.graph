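# Material function for the Lazuli glider that turns an SDF input into an animated mask of repeating wave bands.
# Time and an offset, scaled by wave speed, scroll the SDF, which is wrapped by frequency into distances from evenly spaced line centers.
# Thickness remaps between low and high limits, and a sine of the scrolled field switches it between thinner and thicker variants.
# Screen-space derivatives size the smoothstep edge for antialiased lines.
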MaterialFunction /BRCosmetics/Gadgets/Assets/VinderTech_GliderChute/Glider_Lazuli/Materials/MF_S26Wave_Glider
graph {nodes [
  n1 = FunctionInput() {inputName=WaveThickness_Low, inputType=FunctionInput_Scalar, sortPriority=-50, previewValue=<0.007 0 0 1>}
  n2 = FunctionInput() {inputName=WaveThickness_High, inputType=FunctionInput_Scalar, sortPriority=-49, previewValue=<0.02 0 0 1>}
  n3 = Clamp(None=MaterialFunctionCall(
    Input=FunctionInput() {inputName=Thickness, inputType=FunctionInput_Scalar, sortPriority=-60}
    Input Low=Constant()
    Input High=Constant() {r=1}
    Target Low=@n1
    Target High=@n2
   ) {materialFunction=&RemapValueRange}.Result)
  n4 = FunctionInput() {inputName=WaveThickness_Variance, inputType=FunctionInput_Scalar, sortPriority=-18, previewValue=<0.25 0 0 1>}
  n5 = Time()
  n6 = FunctionInput() {inputName=WaveOffset, inputType=FunctionInput_Scalar, sortPriority=-70}
  n7 = Add(
   A=FunctionInput() {inputName=SDF, inputType=FunctionInput_Scalar, sortPriority=-100}
   B=Multiply(A=Add(A=@n5, B=@n6), B=FunctionInput() {inputName=WaveSpeed, inputType=FunctionInput_Scalar, sortPriority=-80})
  )
  n8 = Sine(None=@n7) {period=0.4}
  n9 = LinearInterpolate(
   A=Multiply(A=@n3, B=Subtract(B=@n4)) {constB=0.75}
   B=Multiply(A=@n3, B=Add(B=@n4) {constA=1}) {constB=1.25}
   Alpha=Ceil(None=Saturate(None=@n8))
  )
  n10 = DDX(Value=@n7)
  n11 = DDY(Value=@n7)
  n12 = Multiply(
   A=Add(A=Abs(None=@n10), B=Abs(None=@n11))
   B=Multiply(A=ScalarParameter() {defaultValue=3, group=DF_Blur, parameterName=DDXDDYMulti}) {constB=0.5}
  ) {constB=0.5}
  n13 = Subtract(A=@n9, B=@n12)
  n14 = Add(A=@n9, B=@n12)
  n15 = FunctionInput() {inputName=WaveFrequency, inputType=FunctionInput_Scalar, sortPriority=-90, previewValue=<0.1 0 0 1>}
  n16 = Frac(None=Add(A=Divide(A=@n7, B=@n15)) {constB=0.5})
  n17 = Multiply(A=Abs(None=Subtract(A=@n16) {constB=0.5}), B=@n15)
  FunctionOutput(None=OneMinus(None=SmoothStep(Min=@n13, Max=@n14, Value=@n17)))
 ]}
