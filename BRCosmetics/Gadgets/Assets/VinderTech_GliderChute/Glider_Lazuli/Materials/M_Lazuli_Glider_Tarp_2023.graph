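# Surface material for the Lazuli glider tarp, painting animated wave-line shapes over the standard 2023 character material.
# A distance-field texture, sampled through scaled and rotated UVs, feeds MF_S26Wave_Glider to form line masks and shape colors.
# Base color, metallic, specular, roughness and flattened normal are blended into the PreFX attributes, then passed to PostFX.
# A GliderArea tech-art channel mask is defined, alongside leftover unconnected experiment nodes.
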Material /BRCosmetics/Gadgets/Assets/VinderTech_GliderChute/Glider_Lazuli/Materials/M_Lazuli_Glider_Tarp_2023
graph {
 properties {
  bUseMaterialAttributes=true
  bUsedWithMeshDeformer=true
  bUsedWithSkeletalMesh=true
  pixelDepthOffsetMode=PDOM_Legacy
  shadingModel=MSM_FromMaterialExpression
  usedShadingModels=Default Lit | Subsurface | Clear Coat | Subsurface Profile | Cloth
 }
 nodes [
  Constant3Vector() {constant=<rgba 0 0 1 0>}
  n1 = NamedRerouteUsage()
  n2 = TextureSample(UVs=NamedRerouteUsage(), Tex=@n1) {samplerSource=SSM_Wrap_WorldGroupSettings, samplerType=SAMPLERTYPE_LinearColor}
  Divide(A=@n2.R)
  n3 = MaterialFunctionCall() {materialFunction=&MF_CharacterParent_2023_PreFX}
  n4 = GetMaterialAttributes(None=@n3.Result) {attributeGetTypes [
    #1
    #2
    #3
    #4
    #5
   ]}
  n5 = Constant() {r=0.8}
  n6 = SetMaterialAttributes!(
   MaterialAttributes=@n4.MaterialAttributes
   Specular=NamedRerouteUsage()
   Base Color=NamedRerouteUsage()
   Metallic=NamedRerouteUsage()
   Roughness=NamedRerouteUsage()
   Normal=LinearInterpolate(
    A=@n4.Normal
    B=MaterialFunctionCall(Normal=@n4.Normal, Flatness=@n5) {materialFunction=&FlattenNormal}.Result
    Alpha=NamedRerouteUsage()
   )
  ) {attributeSetTypes [
    #5
    #1
    #2
    #3
    #4
   ]}
  MaterialFunctionCall(Material=@n6, M=@n3.M, Pre Skinned Local Position=@n3.M, Distance Blend=@n3.M, Fresnel=@n3.M) {materialFunction=&MF_CharacterParent_2023_PostFX}
  Multiply(A=@n2.R) {constB=2}
  NamedRerouteDeclaration(None=@n3.Tech Art Mask) {name=CachedTechArtMask, nodeColor=<rgba 1 0 0.3281 1>}
  NamedRerouteDeclaration(None=ChannelMaskParameter(None=NamedRerouteUsage()) {group=AdditionalTechArt, parameterName=GliderArea}.R) {name=CachedGliderArea, nodeColor=<rgba 0 1 0.1094 1>}
  MaterialExpressionNamedRerouteUsage_13 = NamedRerouteUsage()
  MaterialExpressionNamedRerouteUsage_12 = NamedRerouteUsage()
  MaterialExpressionAppendVector_0 = AppendVector(
   A=ScalarParameter() {defaultValue=1, group=Glider, parameterName=Shape_UVScale02X, sortPriority=65}
   B=ScalarParameter() {defaultValue=1, group=Glider, parameterName=Shape_UVScale02Y, sortPriority=65}
  )
  n7 = Custom!(
   OriginalUV=@MaterialExpressionNamedRerouteUsage_13
   UVPoint=@MaterialExpressionNamedRerouteUsage_12
   UVScale=@MaterialExpressionAppendVector_0
  ) {
   code="float2 DistUV =OriginalUV-UVPoint;\r\nDistUV = DistUV/(UVScale/2) + float2(0.5,0.5);\r\n\r\nreturn DistUV;"
   description=CopyToPointUV
   outputType=CMOT_Float2
   inputs [
    {
     inputName=OriginalUV
     input {expression=&self:MaterialExpressionNamedRerouteUsage_13}
    }
    {
     inputName=UVPoint
     input {expression=&self:MaterialExpressionNamedRerouteUsage_12}
    }
    {
     inputName=UVScale
     input {expression=&self:MaterialExpressionAppendVector_0}
    }
   ]
  }
  n8 = Clamp(None=@n7)
  n9 = Constant2Vector() {g=0.5, r=0.5}
  n10 = ScalarParameter() {group=Glider, parameterName=Shape_UVRotation02, sortPriority=64}
  n11 = NamedRerouteUsage()
  n12 = Subtract(A=@n11) {constB=0.5}
  n13 = Multiply(
   A=LinearInterpolate(A=OneMinus(None=@n11), B=@n11, Alpha=Step(X=@n12))
   B=ScalarParameter() {defaultValue=0.5, group=Glider, parameterName=PostBlend}
  )
  n14 = ScalarParameter() {defaultValue=2, group=Glider, parameterName=PostBlendPower}
  NamedRerouteDeclaration(None=LinearInterpolate(
    A=MaterialFunctionCall(UVs=@n8, Rotation Center=@n9, "Rotation Angle (0-1)"=@n10) {materialFunction=&CustomRotator}.Rotated Values
    B=NamedRerouteUsage()
    Alpha=Power(Base=@n13, Exp=@n14)
   )) {name=ShapeUV02, nodeColor=<rgba 0 0.7656 1 1>}
  n15 = ScalarParameter() {defaultValue=0.2, group=Glider, parameterName=MetalicMin}
  n16 = ScalarParameter() {defaultValue=0.5, group=Glider, parameterName=MetalicMax}
  n17 = NamedRerouteUsage()
  n18 = NamedRerouteUsage()
  n19 = NamedRerouteDeclaration(None=Saturate(None=Add(A=@n17, B=@n18))) {name=LineArea, nodeColor=<rgba 0.1953 0 1 1>}
  n20 = NamedRerouteUsage()
  NamedRerouteDeclaration(None=LinearInterpolate(A=NamedRerouteUsage(), B=LinearInterpolate(A=@n15, B=@n16, Alpha=@n19) {constA=0.5, constB=0.2}, Alpha=@n20)) {name=MetalOUT, nodeColor=<rgba 0.1484 0 1 1>}
  NamedRerouteDeclaration(None=@n2.R) {name=Red_Shape_SDF, nodeColor=<rgba 0.5703 0 1 1>}
  n21 = MaterialFunctionCall!(
   SDF=NamedRerouteUsage()
   WaveFrequency=NamedRerouteUsage()
   WaveSpeed=NamedRerouteUsage()
   WaveOffset=NamedRerouteUsage()
   Thickness=ScalarParameter() {defaultValue=0.01, group=Glider, parameterName=Red_ShapeLineThickness, sortPriority=54}
   WaveThickness_Low=NamedRerouteUsage()
   WaveThickness_High=NamedRerouteUsage()
   WaveThickness_Variance=ScalarParameter() {defaultValue=0.25, group=Glider, parameterName=ShapeLineVar}
  ) {materialFunction=&/MF_S26Wave_Glider}
  NamedRerouteDeclaration(None=SphereMask(
    A=Saturate(None=@n21.Result)
    B=Constant() {r=1}
    Radius=Constant() {r=1}
    Hardness=ScalarParameter() {defaultValue=1, group=Glider, parameterName=DF_LineHardness_Pink, sliderMax=1}
   ) {attenuationRadius=1}) {name=Pink_ShapeLines, nodeColor=<rgba 0.7578 0 1 1>}
  NamedRerouteDeclaration(None=LinearInterpolate(A=NamedRerouteUsage(), B=NamedRerouteUsage(), Alpha=@n20)) {name=BaseColorOUT, nodeColor=<rgba 1 0 0.6562 1>}
  NamedRerouteDeclaration(None=@n4.Roughness) {name=CachedRoughnessIn, nodeColor=<rgba 1 0.2344 0 1>}
  n22 = Divide(
   A=Subtract(A=PixelDepth().R, B=ScalarParameter() {defaultValue=50, group=Glider, parameterName=BlurStartDistance})
   B=ScalarParameter() {defaultValue=500, group=Glider, parameterName=BlurFadeDistance}
  )
  NamedRerouteDeclaration(None=LinearInterpolate(B=ScalarParameter() {defaultValue=0.5, group=Glider, parameterName=BlurMinThick}, Alpha=Saturate(None=@n22)) {constA=1}) {name=DepthDrivenRadius, nodeColor=<rgba 0.7578 1 0 1>}
  n23 = ScalarParameter() {defaultValue=0.7, group=Glider, parameterName=RoughnessMax}
  n24 = ScalarParameter() {defaultValue=0.2, group=Glider, parameterName=RoughnessMin}
  n25 = NamedRerouteUsage()
  NamedRerouteDeclaration(None=LinearInterpolate(A=NamedRerouteUsage(), B=LinearInterpolate(A=@n23, B=@n24, Alpha=@n19) {constA=0.5, constB=0.2}, Alpha=@n25)) {name=RoughnessOUT, nodeColor=<rgba 0.9688 0 1 1>}
  NamedRerouteDeclaration(None=@n4.Specular) {name=CachedSpecularIn, nodeColor=<rgba 0.4062 0 1 1>}
  n26 = ScalarParameter() {defaultValue=0.4, group=Glider, parameterName=SpecMin}
  n27 = ScalarParameter() {defaultValue=0.7, group=Glider, parameterName=RoughnessMax}
  NamedRerouteDeclaration(None=LinearInterpolate(A=NamedRerouteUsage(), B=LinearInterpolate(A=@n26, B=@n27, Alpha=@n19) {constA=0.5, constB=0.2}, Alpha=@n25)) {name=SpecularOUT, nodeColor=<rgba 0 1 0.3906 1>}
  n28 = VectorParameter() {group=Glider, parameterName=BackGroundColor, sortPriority=40, defaultValue=<rgba 0.0093 0 0.3229 1>}
  n29 = LinearInterpolate(
   A=NamedRerouteDeclaration(None=@n28.RGB) {name=CachedBackGroundColor, nodeColor=<rgba 1 0.6563 0 1>}
   B=VectorParameter() {group=Glider, parameterName=BackPinkColor, sortPriority=40, defaultValue=<rgba 1 0 0 1>}.RGB
   Alpha=NamedRerouteUsage()
  )
  n30 = NamedRerouteDeclaration(None=VectorParameter() {group=Glider, parameterName=LineColor, sortPriority=40, defaultValue=<rgba 0 0 0 1>}.RGB) {name=CachedLineColor, nodeColor=<rgba 0 0.9766 1 1>}
  n31 = NamedRerouteUsage()
  NamedRerouteDeclaration(None=Multiply(
    A=LinearInterpolate(A=@n29, B=@n30, Alpha=@n31)
    B=ScalarParameter() {defaultValue=1, group=Glider, parameterName=ShapeEmisiveStrength, sortPriority=41}
   )) {name=ShapeColor, nodeColor=<rgba 0.8281 1 0 1>}
  NamedRerouteDeclaration(None=ScalarParameter() {defaultValue=-0.65, group=Glider, parameterName=ShapeLinePulseSpeed, sortPriority=60}) {name=ShapeLinePulseSpeed, nodeColor=<rgba 0 0.6686 1 1>}
  NamedRerouteDeclaration(None=TextureSample(UVs=NamedRerouteUsage(), Tex=@n1) {samplerSource=SSM_Wrap_WorldGroupSettings, samplerType=SAMPLERTYPE_LinearColor}.R) {name=Dark_Shape_SDF, nodeColor=<rgba 0.5703 0 1 1>}
  n32 = NamedRerouteDeclaration(None=ScalarParameter() {group=Glider, parameterName=ShapeCenterStep, sortPriority=61}) {name=ShapeCenterStep, nodeColor=<rgba 0.4297 0 1 1>}
  n33 = NamedRerouteUsage()
  n34 = NamedRerouteDeclaration(None=ScalarParameter() {defaultValue=0.2, group=Glider, parameterName=ShapeLineDistance, sortPriority=50}) {name=ShapeLineDistance, nodeColor=<rgba 0 0.6686 1 1>}
  n35 = NamedRerouteDeclaration(None=ScalarParameter() {defaultValue=-0.02, group=Glider, parameterName=ShapeLineSpeed, sortPriority=57}) {name=ShapeLineSpeed, nodeColor=<rgba 0.8516 1 0 1>}
  n36 = NamedRerouteDeclaration(None=ScalarParameter() {group=Glider, parameterName=TimeOffset}) {name=CachedTimeOffset, nodeColor=<rgba 0 0.8594 1 1>}
  n37 = ScalarParameter() {defaultValue=0.1, group=Glider, parameterName=Thickness01}
  n38 = NamedRerouteDeclaration(None=ScalarParameter() {defaultValue=0.007, group=Glider, parameterName=Thickness_Low}) {name=ShapeLineWidthMin, nodeColor=<rgba 0 0.6686 1 1>}
  n39 = NamedRerouteDeclaration(None=ScalarParameter() {defaultValue=0.02, group=Glider, parameterName=Thickness_High}) {name=ShapeLineWidthMax, nodeColor=<rgba 0 0.6686 1 1>}
  n40 = ScalarParameter() {defaultValue=0.25, group=Glider, parameterName=ShapeLineVar}
  n41 = SphereMask(
   A=Saturate(None=MaterialFunctionCall(
     SDF=@n33
     WaveFrequency=@n34
     WaveSpeed=@n35
     WaveOffset=@n36
     Thickness=@n37
     WaveThickness_Low=@n38
     WaveThickness_High=@n39
     WaveThickness_Variance=@n40
    ) {materialFunction=&MF_S26Wave_Glider}.Result)
   B=Constant() {r=1}
   Radius=Constant() {r=1}
   Hardness=ScalarParameter() {defaultValue=1, group=Glider, parameterName=DF_LineHardness_Dark, sliderMax=1}
  ) {attenuationRadius=1}
  NamedRerouteDeclaration(None=Multiply(A=Step(Y=@n32, X=@n41) {constY=0.1}, B=@n41)) {name=ShapeLines, nodeColor=<rgba 0.7578 0 1 1>}
  NamedRerouteDeclaration(None=ScalarParameter() {defaultValue=20, group=Glider, parameterName=BlurMaxForTAA}) {name=DrivenBlur, nodeColor=<rgba 0.0312 0 1 1>}
  MaterialExpressionNamedRerouteDeclaration_2 = NamedRerouteDeclaration(None=StaticSwitchParameter(True=TextureCoordinate() {coordinateIndex=1}, False=TextureCoordinate()) {group=Glider, parameterName=UseUV1}) {name=UVs, nodeColor=<rgba 0 0.3672 1 1>}
  n42 = ScalarParameter() {group=Glider, parameterName=Shape_UVOffsetX01, sortPriority=62}
  n43 = ScalarParameter() {group=Glider, parameterName=Shape_UVOffsetY01, sortPriority=63}
  n44 = Constant2Vector() {g=0.5, r=0.5}
  MaterialExpressionNamedRerouteDeclaration_6 = NamedRerouteDeclaration(None=Add(A=AppendVector(A=@n42, B=@n43), B=@n44)) {name=CachedUVPivot, nodeColor=<rgba 0 1 0.6484 1>}
  MaterialExpressionScalarParameter_47 = ScalarParameter() {defaultValue=1, group=Glider, parameterName=Shape_UVScale01, sortPriority=65}
  n45 = Custom(
   OriginalUV=@MaterialExpressionNamedRerouteDeclaration_2
   UVPoint=@MaterialExpressionNamedRerouteDeclaration_6
   UVScale=@MaterialExpressionScalarParameter_47
  ) {
   code="float2 DistUV =OriginalUV-UVPoint;\r\nDistUV = DistUV/(UVScale/2) + float2(0.5,0.5);\r\n\r\nreturn DistUV;"
   description=CopyToPointUV
   outputType=CMOT_Float2
   inputs [
    {
     inputName=OriginalUV
     input {expression=&self:MaterialExpressionNamedRerouteDeclaration_2}
    }
    {
     inputName=UVPoint
     input {expression=&self:MaterialExpressionNamedRerouteDeclaration_6}
    }
    {
     inputName=UVScale
     input {expression=&self:MaterialExpressionScalarParameter_47}
    }
   ]
  }
  NamedRerouteDeclaration(None=MaterialFunctionCall(
    UVs=Clamp(None=@n45)
    Rotation Center=@n44
    "Rotation Angle (0-1)"=ScalarParameter() {group=Glider, parameterName=Shape_UVRotation01, sortPriority=64}
   ) {materialFunction=&CustomRotator}.Rotated Values) {name=ShapeUV01, nodeColor=<rgba 0 0.7656 1 1>}
  NamedRerouteDeclaration(None=@n4.Base Color) {name=CachedBaseColor, nodeColor=<rgba 0 0.25 1 1>}
  n46 = TextureObjectParameter(UVs=@MaterialExpressionNamedRerouteDeclaration_2) {
   group=Glider
   parameterName=Shape01
   samplerType=SAMPLERTYPE_LinearColor
   sortPriority=-10
   texture=/BRCosmetics/Gadgets/Assets/VinderTech_GliderChute/Glider_Lazuli/Textures/T_GliderLazuli_DF
  }
  NamedRerouteDeclaration(None=@n46) {name=CachedShapeTexture01, nodeColor=<rgba 1 0.8203 0 1>}
  NamedRerouteDeclaration(None=@n4.Metallic) {name=CachedMetalicIn, nodeColor=<rgba 0 0.9297 1 1>}
  NamedRerouteUsage()
  NamedRerouteUsage()
  NamedRerouteUsage()
  Power(Base=@n2.R)
  n47 = ScalarParameter() {defaultValue=0.55, group=Experiment, parameterName=AddOffset01}
  SmoothStep(Min=@n47, Max=ScalarParameter() {defaultValue=0.6, group=Experiment, parameterName=StepOffset01}, Value=Add(B=@n47) {constB=0.52})
  n48 = LinearInterpolate(
   A=ScalarParameter() {defaultValue=0.55, group=Experiment, parameterName=AddOffset02}
   B=ScalarParameter() {defaultValue=0.55, group=Experiment, parameterName=AddOffset01}
   Alpha=NamedRerouteUsage()
  )
  SmoothStep(Min=@n48, Max=ScalarParameter() {defaultValue=0.6, group=Experiment, parameterName=StepOffset02}, Value=Add(A=@n48) {constB=0.52})
  StaticBoolParameter() {group=BatEar, parameterName=Ear_ShapeLineVariableThickness, sortPriority=51}
  StaticBoolParameter() {defaultValue=true, group=Glider, parameterName=Red_ShapeLineAnimate, sortPriority=56}
  StaticBoolParameter() {defaultValue=true, group=Glider, parameterName=Red_ShapeLinePulse, sortPriority=58}
 ]
}
